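# Translucent UI master material for the ReactorGrade databit graphic, with the texture and colors exposed as parameters.
# Texture channels act as masks: ToBlur blends the alpha-scaled red (main) and green (blur) channels.
# Blue, less a small offset, adds glow scaled by GlowAmount.
# The saturated mask blends FalloffColor into Color and, multiplied by FinalAlpha, also forms an alpha value.
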Material /PotatoCoaster/Items/ReactorGrade/UI/Art/M_UI_ReactorGrade_Databit_Master
graph {
 properties {blendMode=BLEND_Translucent, materialDomain=MD_UI, pixelDepthOffsetMode=PDOM_Legacy, usedShadingModels=Default Lit}
 nodes [
  n1 = TextureSampleParameter2D(UVs=TextureCoordinate()) {parameterName=Texture, samplerType=SAMPLERTYPE_LinearColor, texture=&/T_UI_ReactorGrade_Databit_1}
  n2 = Multiply(A=@n1.R, B=ScalarParameter() {defaultValue=1, parameterName=MainAlpha})
  n3 = Multiply(A=@n1.G, B=ScalarParameter() {defaultValue=1, parameterName=BlurAlpha})
  n4 = ScalarParameter() {parameterName=ToBlur}
  n5 = ScalarParameter() {parameterName=GlowAmount}
  n6 = Subtract(A=@n1.B, B=ScalarParameter() {defaultValue=0.0125, parameterName=SubtractBlur})
  n7 = Saturate(None=Add(A=LinearInterpolate(A=@n2, B=@n3, Alpha=@n4), B=Multiply(A=@n6, B=@n5)))
  LinearInterpolate(
   A=VectorParameter() {parameterName=FalloffColor, defaultValue=<rgba 0 1 0 1>}.RGB
   B=VectorParameter() {parameterName=Color, defaultValue=<rgba 1 0 0 1>}.RGB
   Alpha=@n7
  )
  Multiply(A=@n7, B=ScalarParameter() {defaultValue=1, parameterName=FinalAlpha})
 ]
}
